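; B_Post_Explosion_Generic is a Blueprint for a generic explosion post-process effect.
; Its Construction Script, the only logic shown, sets the world scale of the PostFX sphere component from the PostFX_PreviewScale variable.
; It then copies the PostFX_Settings variable into the PostFX post-process component's settings, so the effect's look and sphere size are driven by variables.
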(fn ConstructionScript
  (Transformation|SetWorldScale3D (Variables|Default|GetPostFX_SphereComponent) (Variables|Effects|GetPostFX_PreviewScale))
  (Variables|/Script/Engine.PostProcessComponent|SetSettings (Variables|Effects|GetPostFX_Settings) (Variables|Default|GetPostFX)))
