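; Blueprint for a grenade explosion post-process effect, spawned through its custom SpawnPostFX event with box scale, camera shake location, shake toggle and deactivate delay.
; The event passes the scale and shake settings to the shared PostFXStage01 step, then plays a timeline from the start.
; Each timeline update feeds its blend weight and the deactivate delay into PostFXStage02.
; The construction script only calls the parent version.
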(event Custom|SpawnPostFX
  (params (PostFX_BoxScaleDyn :type Float%20%28double-precision%29) (PostFX_CamShakeLoc :type Vector) (PostFX_UseCamShake :type Boolean) (PostFX_DeactivateDelay :type Float%20%28double-precision%29))
  (Utilities|PostFXStage01 PostFX_BoxScaleDyn PostFX_CamShakeLoc PostFX_UseCamShake
    (:Out
      (bind (_direction _blendweight) (Timeline|Timeline_0 PlayFromStart)
        (:Update
          (Utilities|PostFXStage02 _blendweight PostFX_DeactivateDelay))))))

(fn ConstructionScript
  (|CallParentFunction))
